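# Translucent UI material for the icon in the bonus info mini tag, with a swappable Icon texture masked to a circle.
# UVs take a vertical IconHeight offset and center-based IconScale before sampling the clamped texture.
# The sampled alpha passes through ClampAlphaToTransparentBorder, then multiplies by CircleMask, a procedural soft-edged disc built from an inverted SmoothStep of unshifted UV distance to center.
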Material /Game/UI/Frontend/Misc/BonusInfoMiniTag/Materials/M_UI_BonusInfoTagIcon
graph {
 properties {
  bUsedWithStaticMesh=false
  blendMode=BLEND_Translucent
  materialDomain=MD_UI
  pixelDepthOffsetMode=PDOM_Legacy
  usedShadingModels=Default Lit
 }
 nodes [
  n1 = TextureCoordinate()
  n2 = Constant()
  n3 = ScalarParameter() {parameterName=IconHeight}
  n4 = MaterialFunctionCall(
   UVs=Add(A=ComponentMask(None=@n1) {g=true, r=true}, B=AppendVector(A=@n2, B=@n3))
   Texture Scale=ScalarParameter() {defaultValue=0.8, parameterName=IconScale}
  ) {materialFunction=&ScaleUVsByCenter}
  n5 = TextureSampleParameter2D(UVs=@n4.UVs) {
   parameterName=Icon
   samplerSource=SSM_Clamp_WorldGroupSettings
   texture=/Game/Global/Textures/Default/Blanks/T_Blank_White_ZeroAlpha_sRGBColor
  }
  Multiply(
   A=MaterialFunctionCall!(UVs=@n4.UVs, Alpha=@n5.A) {materialFunction=/Game/UI/Foundation/Items/ClampAlphaToTransparentBorder}.Result
   B=NamedRerouteUsage()
  )
  n6 = ComponentMask(None=TextureCoordinate()) {g=true, r=true}
  n7 = Constant() {r=0.5}
  NamedRerouteDeclaration(None=SmoothStep(Value=Distance(A=@n6, B=@n7)) {constMax=0.49, constMin=0.5, constValue=0.5}) {name=CircleMask, nodeColor=<rgba 0 0 0 1>}
 ]
}
