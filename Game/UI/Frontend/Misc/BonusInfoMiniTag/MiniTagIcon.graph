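# MiniTagIcon is a frontend UI widget that shows a configurable icon over a circular backplate for bonus info mini tags.
# PreConstruct or a Refresh event rebuilds the Icon image's brush from the bonus tag icon material at the configured size.
# It then pushes the texture, scale and height variables into that brush's dynamic material.
# The IsAnimIcon and UseBackgroundGradient variables are not read in the graph.
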WidgetBlueprint /Game/UI/Frontend/Misc/BonusInfoMiniTag/MiniTagIcon
graphs [EventGraph {nodes [
   n1 = Refresh() {is=title K2Node_CustomEvent}
   n2 = Get Icon() -> (Icon:Image Object Reference)
   SetBrush(
    execute=[@n1.then, @n4.then]
    self=@n2
    InBrush=Make Slate Brush!(
     TintColor:Slate Color Structure=(SpecifiedColor=(R=1,G=1,B=1,A=1),ColorUseRule=UseColor_Specified)
     DrawAs:ESlateBrushDrawType Enum=Image
     Tiling:ESlateBrushTileType Enum=NoTile
     Mirroring:ESlateBrushMirrorType Enum=NoMirror
     ImageSize:Vector 2D Structure=Get Size() -> (Size:Vector 2D Structure)
     Margin:Margin Structure=(Left=0,Top=0,Right=0,Bottom=0)
     ResourceObject:Object Reference=/Game/UI/Frontend/Misc/BonusInfoMiniTag/Materials/M_UI_BonusInfoTagIcon
     OutlineSettings:Slate Brush Outline Settings Structure=(CornerRadii=(X=0,Y=0,Z=0,W=0),Color=(SpecifiedColor=(R=0,G=0,B=0,A=0),ColorUseRule=UseColor_Specified),Width=0,RoundingType=HalfHeightRadius,bUseBrushTransparency=False)
    ) -> (SlateBrush:Slate Brush Structure) {is=title K2Node_MakeStruct}
   )
   > n3 = GetDynamicMaterial(self=@n2)
   > SetScalarParameterValue(self=@n3, ParameterName=IconScale, Value=Get IconScale() -> (IconScale:double))
   > SetTextureParameterValue(self=@n3, ParameterName=Icon, Value=Get LegoIcon() -> (LegoIcon:Texture 2D Object Reference))
   > SetScalarParameterValue(self=@n3, ParameterName=IconHeight, Value=Get IconHeight() -> (IconHeight:double))
   n4 = Event PreConstruct()
  ]}]
variables [
 IconHeight {type=number, default=0.8}
 IconScale {type=number, default=0.8}
 IsAnimIcon {type=boolean}
 LegoIcon {type=Texture2D}
 Size {type=Vector2D, default=<75 75>}
 UseBackgroundGradient {type=boolean}
]
widgets {
 info {inheritedWidgetCount=0, parentClass=&UserWidget, rootWidgetClass=&Overlay, widgetCount=3}
 widgets [
  Overlay_0 {widgetClassPath=&Overlay, widget=Overlay_0}
  WBP_UIKit_Backplate {
   widgetClassPath=&WBP_UIKit_Backplate
   bIsVariable=true
   parent=Overlay_0
   widget=WBP_UIKit_Backplate
   properties {brightness=Medium, cornerRadius=Circle}
   slot {class=OverlaySlot, properties {horizontalAlignment=HAlign_Fill, verticalAlignment=VAlign_Fill}}
  }
  Icon {
   widgetClassPath=&Image
   bIsVariable=true
   parent=Overlay_0
   widget=Icon
   properties {
    visibility=HitTestInvisible
    brush {
     bIsDynamicallyLoaded=false
     drawAs=Image
     imageType=NoImage
     mirroring=NoMirror
     resourceName=None
     resourceObject=&M_UI_BonusInfoTagIcon
     tiling=NoTile
     imageSize=<75 75>
     margin=<ltrb 0 0 0 0>
     outlineSettings {
      bUseBrushTransparency=false
      roundingType=HalfHeightRadius
      width=0
      color {colorUseRule=UseColor_Specified, specifiedColor=<rgba 0 0 0 0>}
      cornerRadii=<0 0 0 0>
     }
     tintColor {colorUseRule=UseColor_Specified, specifiedColor=<rgba 1 1 1 1>}
     uVRegion {bIsValid=false, max=<0 0>, min=<0 0>}
    }
   }
   slot {class=OverlaySlot}
  }
 ]
}
