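# Frontend mini tag widget showing a cosmetic's bonus LEGO style, Rocket League availability or Festival music, with its claimed state.
# OnSetBonusInfo shows the LEGO or Rocket panel for Juno or DelMar and hides the tag for BattleRoyale and Sparks, leaving the Festival panel unreached.
# SetBonusInfoAbstraction instead reads an item viewmodel and soft-loads its icon.
# UpdateUI applies mobile padding, colors, claimed dimming, switcher index and formatted names.
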WidgetBlueprint /Game/UI/Frontend/Misc/BonusInfoMiniTag/BonusInfoMiniTag
graphs [
 EventGraph {nodes [
   n1 = RequestLoadIcon!() -> (OutputDelegate, then, SoftIconAsset:Texture 2D Soft Object Reference) {is=title K2Node_CustomEvent}
   > n2 = Load From Soft Texture!(execute:Exec, Texture:Texture 2D Soft Object Reference=@n1.SoftIconAsset) -> (then:Exec, Completed:Exec, Result:Texture 2D Object Reference) {is=title MVVMK2Node_LoadSoftTexture}
   Set LegoIcon2D(execute=@n2.Completed, LegoIcon2D:Texture 2D Object Reference=@n2) -> (Output_Get:Texture 2D Object Reference)
   > UpdateUI()
   UpdateUI() {is=title K2Node_CustomEvent}
   > SetMobileStyle() {is=title K2Node_CallFunction}
   > UpdateTextColors()
   n3 = Get Claimed() -> (Claimed:bool)
   > SetColorAndOpacity(
    self:User Widget Object Reference
    InColorAndOpacity:Linear Color Structure=Make Linear Color!(R:float=1, G:float=1, B:float=1, A:float=SelectFloat(A=0.5, B=1, bPickA=@n3)) -> (LinearColor:Linear Color Structure) {is=title K2Node_MakeStruct}
   )
   > SetActiveWidgetIndex(
    self:Common Visibility Switcher Object Reference=Get VisibilitySwitcher() -> (VisibilitySwitcher:Common Visibility Switcher Object Reference)
    Index=Get SwitcherIndex() -> (SwitcherIndex:int=0)
   )
   n4 = Get Lego_MiniTagIcon() -> (Lego_MiniTagIcon:Mini Tag Icon Object Reference)
   > Set LegoIcon!(
    execute
    LegoIcon:Texture 2D Object Reference=Get LegoIcon2D() -> (LegoIcon2D:Texture 2D Object Reference)
    self:Mini Tag Icon Object Reference=@n4
   ) -> (then, Output_Get:Texture 2D Object Reference)
   > Refresh(self:Mini Tag Icon Object Reference=@n4)
   > SetVisibility(
    self=[
     Get Rocket_Claimed() -> (Rocket_Claimed:Common Text Object Reference)
     Get Festival_AlreadyOwned() -> (Festival_AlreadyOwned:Common Text Object Reference)
     Get Lego_AlreadyOwned() -> (Lego_AlreadyOwned:Common Text Object Reference)
    ]
    InVisibility=Select(Option 0=Collapsed, Option 1=Visible, Index=Get Claimed())
   )
   > SetVisibility(
    self=[
     Get Festival_Includes() -> (Festival_Includes:Common Text Object Reference)
     Get Lego_Includes() -> (Lego_Includes:Common Text Object Reference)
     Get Rocket_ToClaim() -> (Rocket_ToClaim:Common Text Object Reference)
    ]
    InVisibility=Select(Option 0=Visible, Option 1=Collapsed, Index=Get Claimed())
   )
   > UpdateOutfitNameText()
   > UpdateMusicNameText()
   n5 = Event OnSetBonusInfo!() -> (
    OutputDelegate
    then
    BonusInfoType:"ECosmeticCompatibleMode Enum (by ref)"=BattleRoyale
    bClaimed:bool
    Description:"Text (by ref)"
    PreviewTexture:Texture 2D Object Reference
    bIsEmote:bool
   ) {is=title K2Node_Event}
   > n6 = Set InfoType(InfoType:ECosmeticCompatibleMode Enum=@n5.BonusInfoType) -> (Output_Get:ECosmeticCompatibleMode Enum=BattleRoyale)
   > n7 = Switch on ECosmeticCompatibleMode!(execute, Selection:ECosmeticCompatibleMode Enum=@n6) -> (BattleRoyale:Exec, Juno:Exec, DelMar:Exec, Sparks:Exec) {is=title K2Node_SwitchEnum}
   SetVisibility(execute=[@n7.BattleRoyale, @n7.Sparks], InVisibility=Collapsed)
   n8 = Set SwitcherIndex(execute=@n7.DelMar, SwitcherIndex:int=1) -> (Output_Get:int=0)
   SetVisibility(execute=[@n12.then, @n11.then, @n8.then], InVisibility=Visible)
   > Set Claimed(Claimed:bool=@n5.bClaimed) -> (Output_Get:bool)
   > n9 = Is Valid(InputObject=@n5.PreviewTexture)
   > n10 = Set LegoIcon2D(LegoIcon2D=@n5.PreviewTexture)
   Set RewardName(execute=[@n10.then, @n9.Is Not Valid], RewardName:Text=@n5.Description) -> (Output_Get:Text)
   > UpdateUI() {is=title K2Node_CallFunction}
   n11 = Set SwitcherIndex(execute=@n7.Juno, SwitcherIndex=0) -> (Output_Get=0)
   Event PreConstruct()
   > UpdateUI()
   n12 = Set SwitcherIndex(SwitcherIndex=2) -> (Output_Get=0)
  ]}
 SetBonusInfoAbstraction {nodes [
   n1 = SetBonusInfoAbstraction!() -> (then, FortItemVM:Fort Item VM Object Reference) {is=title K2Node_FunctionEntry}
   > Set Claimed(Claimed=Get bIsOwned!(self:Fort Item VM Object Reference=@n1) -> (bIsOwned:bool))
   n2 = Get FortItemVM() -> (FortItemVM:Fort Item VM Object Reference)
   > Set RewardName(RewardName=GetDisplayName!(self:Fort Item Details VM Object Reference=Get ItemDetails!(self:Fort Item VM Object Reference=@n2) -> (ItemDetails:Fort Item Details VM Object Reference)) -> (ReturnValue:Text))
   > Set InfoType(InfoType=BattleRoyale) -> (Output_Get=BattleRoyale)
   n3 = Get FortItemVM()
   > RequestLoadIcon!(
    execute
    self
    SoftIconAsset:Texture 2D Soft Object Reference=Get SmallPreviewImage!(self:Fort Item Details VM Object Reference=Get ItemDetails(self=@n3)) -> (SmallPreviewImage:Texture 2D Soft Object Reference)
   ) -> (then) {is=title K2Node_CallFunction}
  ]}
 SetMobileStyle {nodes [
   SetMobileStyle() {is=title K2Node_FunctionEntry}
   > Branch(Condition=IsMobileGame())
   > SetPadding!(
    execute
    self:Overlay Slot Object Reference=SlotAsOverlaySlot(Widget=Get VisibilitySwitcher())
    InPadding_Left:float=0
    InPadding_Top:float=24
    InPadding_Right:float=0
    InPadding_Bottom:float=0
   ) -> (then)
  ]}
 UpdateMusicNameText {nodes [
   UpdateMusicNameText() {is=title K2Node_FunctionEntry}
   n1 = Get InfoType() -> (InfoType:ECosmeticCompatibleMode Enum=BattleRoyale)
   n2 = TextToUpper!(InText:"Text (by ref)"=Get RewardName() -> (RewardName:Text)) -> (ReturnValue:Text)
   > Set Text(
    Text=SelectText(
     A=Format Text!(Format="LOCTEXT(\"{0} JAM TRACK\")", 0:Text=@n2) -> (Result)
     B=Format Text(Format="LOCTEXT(\"{0} LOBBY MUSIC\")", 0=@n2)
     bPickA="Equal (Enum)"(A:ECosmeticCompatibleMode Enum=@n1, B:ECosmeticCompatibleMode Enum=Sparks)
    )
    self=Get Festival_SongName() -> (Festival_SongName:Common Text Object Reference)
   )
  ]}
 UpdateOutfitNameText {nodes [
   UpdateOutfitNameText() {is=title K2Node_FunctionEntry}
   n1 = Get InfoType() -> (InfoType=BattleRoyale)
   n2 = TextToUpper(InText=Get RewardName())
   > Set Text(
    Text=SelectText(
     A=Format Text(Format="LOCTEXT(\"{0} LEGO STYLE\")", 0=@n2)
     B=Format Text(Format="LOCTEXT(\"{0} FORTNITE STYLE\")", 0=@n2)
     bPickA="Equal (Enum)"(A=@n1, B=Juno)
    )
    self=Get Lego_OutfitName() -> (Lego_OutfitName:Common Text Object Reference)
   )
  ]}
 UpdateTextColors {nodes [
   UpdateTextColors() {is=title K2Node_FunctionEntry}
   > SetColorAndOpacity!(
    execute
    self:Text Object Reference=[
     Get Festival_AlreadyOwned()
     Get Festival_Includes()
     Get Festival_SongName()
     Get Lego_AlreadyOwned()
     Get Lego_Includes()
     Get Lego_OutfitName()
     Get Rocket_Claimed()
     Get Rocket_ToClaim()
    ]
    InColorAndOpacity_SpecifiedColor:Linear Color Structure=Get TextColor() -> (TextColor:Linear Color Structure)
    InColorAndOpacity_ColorUseRule:ESlateColorStylingMode Enum
   ) -> (then)
  ]}
]
variables [
 Claimed {type=boolean}
 InfoType {type=ECosmeticCompatibleMode, default=BattleRoyale}
 LegoIcon2D {type=Texture2D}
 RewardName {type=string, default=""}
 SwitcherIndex {type=integer}
 TextColor {type=LinearColor, default=<rgba 1 1 1 1>}
]
widgets {
 info {inheritedWidgetCount=0, parentClass=/Script/FortniteUI.FortBonusInfoMiniTag, rootWidgetClass=&Overlay, widgetCount=19}
 widgets [
  Overlay_0 {widgetClassPath=&Overlay, widget=Overlay_0}
  VisibilitySwitcher {
   widgetClassPath=&CommonVisibilitySwitcher
   bIsVariable=true
   parent=Overlay_0
   widget=VisibilitySwitcher
   slot {class=OverlaySlot, properties {horizontalAlignment=HAlign_Fill, padding=<ltrb 0 7.5 0 0>}}
  }
  Lego {widgetClassPath=&Overlay, parent=VisibilitySwitcher, widget=Lego, slot {class=CommonVisibilitySwitcherSlot}}
  Lego_MiniTagIcon {
   widgetClassPath=&/MiniTagIcon
   bIsVariable=true
   parent=Lego
   widget=Lego_MiniTagIcon
   properties {iconHeight=0, iconScale=1, useBackgroundGradient=true}
   slot {class=OverlaySlot, properties {verticalAlignment=VAlign_Center, padding=<ltrb 0 5 5 5>}}
  }
  VB_Lego_Texts {
   widgetClassPath=&VerticalBox
   parent=Lego
   widget=VB_Lego_Texts
   properties {clipping=ClipToBounds}
   slot {class=OverlaySlot, properties {horizontalAlignment=HAlign_Fill, verticalAlignment=VAlign_Center, padding=<ltrb 90 0 0 0>}}
  }
  Lego_Includes {
   widgetClassPath=&CommonTextBlock
   bIsVariable=true
   parent=VB_Lego_Texts
   widget=Lego_Includes
   properties {
    bAutoCollapseWithEmptyText=true
    text=INCLUDES:
    wrapTextAt=550
    colorAndOpacity {colorUseRule=UseColor_Specified, specifiedColor=<rgba 0.2462 0.0052 0.0976 1>}
    font {
     bForceMonospaced=false
     bMaterialIsStencil=false
     fontMaterial=null
     fontObject=&HeadingNow
     letterSpacing=0
     monospacedWidth=1
     size=14.25
     skewAmount=0
     typefaceFontName=76 Bold Italic
     outlineSettings {
      bApplyOutlineToDropShadows=false
      bMiteredCorners=false
      bSeparateFillAlpha=false
      outlineMaterial=null
      outlineSize=0
      outlineColor=<rgba 0 0 0 1>
     }
    }
    renderTransform {angle=0, scale=<1 1>, shear=<-5 0>, translation=<0 0>}
    shadowOffset=<0 0>
   }
   slot {class=VerticalBoxSlot}
  }
  Lego_OutfitName {
   widgetClassPath=&CommonTextBlock
   bIsVariable=true
   parent=VB_Lego_Texts
   widget=Lego_OutfitName
   properties {
    bAutoCollapseWithEmptyText=true
    scrollStyle=&TextScrollStyle-Base_C
    text=" LEGO® STYLE "
    colorAndOpacity {colorUseRule=UseColor_Specified, specifiedColor=<rgba 0.2462 0.0052 0.0976 1>}
    font {
     bForceMonospaced=false
     bMaterialIsStencil=false
     fontMaterial=null
     fontObject=&HeadingNow
     letterSpacing=0
     monospacedWidth=1
     size=14.25
     skewAmount=0
     typefaceFontName=76 Bold Italic
     outlineSettings {
      bApplyOutlineToDropShadows=false
      bMiteredCorners=false
      bSeparateFillAlpha=false
      outlineMaterial=null
      outlineSize=0
      outlineColor=<rgba 0 0 0 1>
     }
    }
    renderTransform {angle=0, scale=<1 1>, shear=<-5 0>, translation=<0 0>}
    shadowOffset=<0 0>
   }
   slot {class=VerticalBoxSlot}
  }
  Lego_AlreadyOwned {
   widgetClassPath=&CommonTextBlock
   bIsVariable=true
   parent=VB_Lego_Texts
   widget=Lego_AlreadyOwned
   properties {
    bAutoCollapseWithEmptyText=true
    text=ALREADY OWNED
    wrapTextAt=550
    font {
     bForceMonospaced=false
     bMaterialIsStencil=false
     fontMaterial=null
     fontObject=&HeadingNow
     letterSpacing=0
     monospacedWidth=1
     size=14.25
     skewAmount=0
     typefaceFontName=76 Bold Italic
     outlineSettings {
      bApplyOutlineToDropShadows=false
      bMiteredCorners=false
      bSeparateFillAlpha=false
      outlineMaterial=null
      outlineSize=0
      outlineColor=<rgba 0 0 0 1>
     }
    }
    renderTransform {angle=0, scale=<1 1>, shear=<-5 0>, translation=<0 0>}
    shadowOffset=<0 0>
   }
   slot {class=VerticalBoxSlot}
  }
  RocketRacing {
   widgetClassPath=&Overlay
   parent=VisibilitySwitcher
   widget=RocketRacing
   slot {class=CommonVisibilitySwitcherSlot, properties {verticalAlignment=VAlign_Top}}
  }
  Rocket_MiniTagIcon {
   widgetClassPath=&MiniTagIcon
   bIsVariable=true
   parent=RocketRacing
   widget=Rocket_MiniTagIcon
   properties {iconHeight=0, iconScale=0.9, legoIcon=/Game/UI/Frontend/Misc/BonusInfoMiniTag/Textures/T_UI_S28_RocketIcon}
   slot {class=OverlaySlot, properties {verticalAlignment=VAlign_Center, padding=<ltrb 0 5 5 5>}}
  }
  VB_Rocket_Texts {
   widgetClassPath=&VerticalBox
   parent=RocketRacing
   widget=VB_Rocket_Texts
   slot {class=OverlaySlot, properties {horizontalAlignment=HAlign_Fill, verticalAlignment=VAlign_Center, padding=<ltrb 90 0 0 0>}}
  }
  Rocket_ToClaim {
   widgetClassPath=&CommonTextBlock
   bIsVariable=true
   parent=VB_Rocket_Texts
   widget=Rocket_ToClaim
   properties {
    bAutoCollapseWithEmptyText=true
    text=Unlocking this item will also make it available in Rocket League
    wrapTextAt=400
    font {
     bForceMonospaced=false
     bMaterialIsStencil=false
     fontMaterial=null
     fontObject=&HeadingNow
     letterSpacing=0
     monospacedWidth=1
     size=14.25
     skewAmount=0
     typefaceFontName=76 Bold Italic
     outlineSettings {
      bApplyOutlineToDropShadows=false
      bMiteredCorners=false
      bSeparateFillAlpha=false
      outlineMaterial=null
      outlineSize=0
      outlineColor=<rgba 0 0 0 1>
     }
    }
    renderTransform {angle=0, scale=<1 1>, shear=<-5 0>, translation=<0 0>}
    shadowOffset=<0 0>
   }
   slot {class=VerticalBoxSlot}
  }
  Rocket_Claimed {
   widgetClassPath=&CommonTextBlock
   bIsVariable=true
   parent=VB_Rocket_Texts
   widget=Rocket_Claimed
   properties {
    bAutoCollapseWithEmptyText=true
    text=Item now available in Rocket League
    wrapTextAt=400
    font {
     bForceMonospaced=false
     bMaterialIsStencil=false
     fontMaterial=null
     fontObject=&HeadingNow
     letterSpacing=0
     monospacedWidth=1
     size=14.25
     skewAmount=0
     typefaceFontName=76 Bold Italic
     outlineSettings {
      bApplyOutlineToDropShadows=false
      bMiteredCorners=false
      bSeparateFillAlpha=false
      outlineMaterial=null
      outlineSize=0
      outlineColor=<rgba 0 0 0 1>
     }
    }
    renderTransform {angle=0, scale=<1 1>, shear=<-5 0>, translation=<0 0>}
    shadowOffset=<0 0>
   }
   slot {class=VerticalBoxSlot}
  }
  FortniteFestival {
   widgetClassPath=&Overlay
   parent=VisibilitySwitcher
   widget=FortniteFestival
   slot {class=CommonVisibilitySwitcherSlot, properties {verticalAlignment=VAlign_Top}}
  }
  Festival_MiniTagIcon {
   widgetClassPath=&Image
   bIsVariable=true
   parent=FortniteFestival
   widget=Festival_MiniTagIcon
   properties {brush {
     bIsDynamicallyLoaded=false
     drawAs=Image
     imageType=NoImage
     mirroring=NoMirror
     resourceName=None
     resourceObject=/Game/UI/Frontend/Misc/BonusInfoMiniTag/Textures/T_UI_S28_FestivalIcon
     tiling=NoTile
     imageSize=<75 75>
     margin=<ltrb 0 0 0 0>
     outlineSettings {
      bUseBrushTransparency=false
      roundingType=HalfHeightRadius
      width=0
      color {colorUseRule=UseColor_Specified, specifiedColor=<rgba 0 0 0 0>}
      cornerRadii=<0 0 0 0>
     }
     tintColor {colorUseRule=UseColor_Specified, specifiedColor=<rgba 1 1 1 1>}
     uVRegion {bIsValid=false, max=<0 0>, min=<0 0>}
    }}
   slot {class=OverlaySlot, properties {padding=<ltrb 0 5 5 5>}}
  }
  VB_Festival_Texts {
   widgetClassPath=&VerticalBox
   parent=FortniteFestival
   widget=VB_Festival_Texts
   slot {class=OverlaySlot, properties {horizontalAlignment=HAlign_Fill, verticalAlignment=VAlign_Center, padding=<ltrb 90 0 0 0>}}
  }
  Festival_Includes {
   widgetClassPath=&CommonTextBlock
   bIsVariable=true
   parent=VB_Festival_Texts
   widget=Festival_Includes
   properties {
    bAutoCollapseWithEmptyText=true
    text=INCLUDES:
    wrapTextAt=550
    font {
     bForceMonospaced=false
     bMaterialIsStencil=false
     fontMaterial=null
     fontObject=&HeadingNow
     letterSpacing=0
     monospacedWidth=1
     size=14.25
     skewAmount=0
     typefaceFontName=76 Bold Italic
     outlineSettings {
      bApplyOutlineToDropShadows=false
      bMiteredCorners=false
      bSeparateFillAlpha=false
      outlineMaterial=null
      outlineSize=0
      outlineColor=<rgba 0 0 0 1>
     }
    }
    renderTransform {angle=0, scale=<1 1>, shear=<-5 0>, translation=<0 0>}
    shadowOffset=<0 0>
   }
   slot {class=VerticalBoxSlot}
  }
  Festival_SongName {
   widgetClassPath=&CommonTextBlock
   bIsVariable=true
   parent=VB_Festival_Texts
   widget=Festival_SongName
   properties {
    bAutoCollapseWithEmptyText=true
    text=JAM TRACK
    wrapTextAt=550
    font {
     bForceMonospaced=false
     bMaterialIsStencil=false
     fontMaterial=null
     fontObject=&HeadingNow
     letterSpacing=0
     monospacedWidth=1
     size=14.25
     skewAmount=0
     typefaceFontName=76 Bold Italic
     outlineSettings {
      bApplyOutlineToDropShadows=false
      bMiteredCorners=false
      bSeparateFillAlpha=false
      outlineMaterial=null
      outlineSize=0
      outlineColor=<rgba 0 0 0 1>
     }
    }
    renderTransform {angle=0, scale=<1 1>, shear=<-5 0>, translation=<0 0>}
    shadowOffset=<0 0>
   }
   slot {class=VerticalBoxSlot}
  }
  Festival_AlreadyOwned {
   widgetClassPath=&CommonTextBlock
   bIsVariable=true
   parent=VB_Festival_Texts
   widget=Festival_AlreadyOwned
   properties {
    bAutoCollapseWithEmptyText=true
    text=ALREADY OWNED
    wrapTextAt=550
    font {
     bForceMonospaced=false
     bMaterialIsStencil=false
     fontMaterial=null
     fontObject=&HeadingNow
     letterSpacing=0
     monospacedWidth=1
     size=14.25
     skewAmount=0
     typefaceFontName=76 Bold Italic
     outlineSettings {
      bApplyOutlineToDropShadows=false
      bMiteredCorners=false
      bSeparateFillAlpha=false
      outlineMaterial=null
      outlineSize=0
      outlineColor=<rgba 0 0 0 1>
     }
    }
    renderTransform {angle=0, scale=<1 1>, shear=<-5 0>, translation=<0 0>}
    shadowOffset=<0 0>
   }
   slot {class=VerticalBoxSlot}
  }
 ]
}
